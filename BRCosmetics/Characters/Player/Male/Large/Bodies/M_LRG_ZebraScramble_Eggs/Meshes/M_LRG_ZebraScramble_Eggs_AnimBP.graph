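# Animation Blueprint for the ZebraScramble Eggs large male player body cosmetic, posing its mesh from the character it attaches to.
# Copy Pose From Mesh takes the attached parent's pose and feeds a skin-specific Control Rig, whose result is the output pose.
# The rig receives the emote-playing state through its Is Emoting input.
# The event graph's events are empty, and the Bacon and emote-fix booleans are unused in these graphs.
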AnimBlueprint /BRCosmetics/Characters/Player/Male/Large/Bodies/M_LRG_ZebraScramble_Eggs/Meshes/M_LRG_ZebraScramble_Eggs_AnimBP
target_skeleton=&/M_LRG_ZebraScramble_Eggs_Skeleton
graphs [
 AnimGraph {nodes [
   n1 = Control Rig!(Source=Copy Pose From Mesh() {bUseAttachedParent=true}, Alpha, Is Emoting:bool=Get bIsPlayingEmote()) -> (Pose) {
    alphaBoolBlend {blendOption=Linear}
    controlRigAssetReference {blueprintRigClass=&/M_LRG_ZebraScramble_Eggs_CtrlRig.M_LRG_ZebraScramble_Eggs_CtrlRig_C}
   }
   Output Pose(Result=@n1) {layerGroup=None}
  ]}
 EventGraph {nodes [
   Event BlueprintUpdateAnimation()
   Event OnCharacterCustomizationLoaded()
  ]}
]
variables [
 Is Bacon {type=boolean}
 ZebraScrambleEmoteFix {type=boolean}
]
